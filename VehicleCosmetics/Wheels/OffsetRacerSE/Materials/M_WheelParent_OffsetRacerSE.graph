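# Masked parent material for the OffsetRacerSE vehicle wheel, layering glowing rim and tire effects over MF_WheelParent_Simple, then applying vehicle damage and de-rez transition functions.
# Switchable rim and tire lights combine texture masks with geometric patterns panning in opposite directions, tinted by a custom color and brightness.
# Red vertex color selects tire over rim, both for this emissive and for base color, where the tire gets a masked, Fresnel-shaded custom color.
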Material /VehicleCosmetics/Wheels/OffsetRacerSE/Materials/M_WheelParent_OffsetRacerSE
graph {
 properties {
  bAutomaticallySetUsageInEditor=false
  bUseMaterialAttributes=true
  bUsedWithMeshDeformer=true
  bUsedWithSkeletalMesh=true
  blendMode=BLEND_Masked
  pixelDepthOffsetMode=PDOM_Legacy
  previewMesh=/VehicleCosmetics/Wheels/RL105/Mesh/SK_Wheel_RL105
  usedShadingModels=Default Lit
 }
 nodes [
  n1 = GetMaterialAttributes(None=MaterialFunctionCall() {materialFunction=&MF_WheelParent_Simple}.MaterialAttributes) {attributeGetTypes [#1, #2, #3]}
  n2 = Add(A=@n1.Emissive Color, B=NamedRerouteUsage())
  n3 = TextureSampleParameter2D() {parameterName=OpasityMask, texture=/VehicleCosmetics/Textures/T_Black}
  n4 = VectorParameter() {group=3_Tire Properties, parameterName=TireColor_Custom, defaultValue=<rgba 0 0.8328 1 1>}
  n5 = Multiply(
   A=MaterialFunctionCall!(
    Input Normal=TextureSample() {samplerType=SAMPLERTYPE_Normal, texture=/VehicleCosmetics/Wheels/OffsetRacerSE/Textures/T_OffsetRacerSe_Tire_N}.RGB
    Input Sharpness=Constant() {r=2}
    Input Two Sided
   ) {materialFunction=/VehicleCosmetics/Wheels/CarCar_Holographic/Materials/MF_Fresnel_Advanced}.Middle
   B=Constant() {r=1}
  )
  n6 = TextureSample() {samplerType=SAMPLERTYPE_LinearColor, texture=/VehicleCosmetics/Wheels/OffsetRacerSE/Textures/T_OffsetRacerSe_Tire_M_2}
  n7 = LinearInterpolate(A=NamedRerouteUsage(), B=LinearInterpolate(A=Multiply(A=@n4.RGB, B=@n5), B=Constant(), Alpha=OneMinus(None=@n6.G)), Alpha=VertexColor().R)
  MaterialFunctionCall(In=MaterialFunctionCall(Material=SetMaterialAttributes!(MaterialAttributes=@n1.MaterialAttributes, Emissive Color=@n2, Opacity Mask=@n3.A, Base Color=@n7) {attributeSetTypes [#1, #2, #3]}) {materialFunction=&MF_BaseVehicle_Damage}.Result) {materialFunction=&MF_VehicleDeRez_Transition}
  n8 = TextureSampleParameter2D() {
   group=5_Special Effects
   parameterName=Emissive_Masks
   samplerSource=SSM_Wrap_WorldGroupSettings
   samplerType=SAMPLERTYPE_LinearColor
   texture=/VehicleCosmetics/Wheels/OffsetRacerSE/Textures/T_OffsetRacerSe_Rim_M
  }
  n9 = Add(A=TextureCoordinate(), B=Constant2Vector() {r=0.4})
  n10 = Panner(Coordinate=@n9) {speedX=0.4}
  n11 = AppendVector(A=ScalarParameter() {group=5_Special Effects, parameterName=Offset_RimEmissive}, B=Constant())
  n12 = Multiply(
   A=@n8.A
   B=TextureSample(UVs=Add(A=@n10, B=@n11)) {samplerSource=SSM_Wrap_WorldGroupSettings, texture=/VehicleCosmetics/Wheels/_SharedTextures/FX/Skin_GeometricBM_RGB}.R
  )
  n13 = Constant()
  n14 = StaticSwitchParameter(True=Add(A=Clamp(None=@n12) {maxDefault=3}, B=@n8.G), False=@n13) {group=5_Special Effects, parameterName=Rim_Lights}
  n15 = Multiply(
   A=TextureSample(UVs=Panner(Coordinate=@n9) {speedX=-0.4}) {samplerSource=SSM_Wrap_WorldGroupSettings, texture=&Skin_GeometricBM_RGB}.R
   B=@n8.B
  )
  n16 = TextureSampleParameter2D() {
   group=5_Special Effects
   parameterName=Normal
   samplerSource=SSM_Wrap_WorldGroupSettings
   samplerType=SAMPLERTYPE_Normal
   texture=/VehicleCosmetics/Wheels/OffsetRacerSE/Textures/T_OffsetRacerSe_Rim_N
  }
  n17 = StaticSwitchParameter(True=Add(A=Clamp(None=@n15) {maxDefault=3}, B=OneMinus(None=@n16.A)), False=@n13) {group=5_Special Effects, parameterName=Tire_Lightd}
  n18 = VertexColor()
  n19 = VectorParameter() {group=5_Special Effects, parameterName=CustomColor, defaultValue=<rgba 0.0999 0.3916 1 1>}
  n20 = ScalarParameter() {defaultValue=3, group=5_Special Effects, parameterName=GlowBrightness}
  NamedRerouteDeclaration(None=Multiply(A=LinearInterpolate(A=@n14, B=@n17, Alpha=@n18.R), B=Multiply(A=@n19.RGB, B=@n20))) {name=Emissive, nodeColor=<rgba 1 0.75 0 1>}
  NamedRerouteDeclaration(None=@n1.Base Color) {name=BaseColor_Input, nodeColor=<rgba 1 0.5859 0 1>}
  TextureCoordinate()
 ]
}
